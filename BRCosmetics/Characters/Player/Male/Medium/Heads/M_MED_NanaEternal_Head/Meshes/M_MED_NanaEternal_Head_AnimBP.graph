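# Animation Blueprint for the NanaEternal male medium head cosmetic, targeting that head's custom Fortnite base head skeleton.
# Its event graph holds only a pawn-owner lookup and a parent Update Animation call, with no event wiring shown.
# It declares no variables, and no AnimGraph or other animation logic appears in the file.
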AnimBlueprint /BRCosmetics/Characters/Player/Male/Medium/Heads/M_MED_NanaEternal_Head/Meshes/M_MED_NanaEternal_Head_AnimBP
target_skeleton=/BRCosmetics/Characters/Player/Male/Medium/Heads/M_MED_NanaEternal_Head/Meshes/CustomSkeleton/Fortnite_Base_Head_NanaEternal_Skeleton
graphs [EventGraph {nodes [
   TryGetPawnOwner()
   Parent: BlueprintUpdateAnimation!(execute, DeltaTimeX:float=0) -> (then) {is=title K2Node_CallParentFunction}
  ]}]
variables []
